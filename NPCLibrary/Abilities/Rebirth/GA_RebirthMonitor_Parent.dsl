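; Parent monitor ability that triggers an AI pawn's Rebirth ability when that pawn enters DBNO (down but not out).
; On activation it caches the avatar as a Fort player pawn and binds to OnPlayerStartDBNO, ending if the cast fails.
; When DBNO fires, it tries to activate abilities matching the Rebirth tag, logs any failure, and ends itself.
; Ending the ability unbinds the DBNO listener.
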(event Ability|EventActivateAbility
  (bind _outfortplayerpawn (CallFunction|GetAvatarasValidFortPlayerPawn)
    (:ValidFortPlayerPawn
      (bind _output_get (Variables|InternalData|SetAIPawn _outfortplayerpawn))
      (EventDispatchers|BindEventtoOnPlayerStartDBNO _output_get (EventDispatchers|CreateEvent :!event Custom|PawnEntredDBNO)))
    (:AvatarCastFailed
      (Ability|EndAbility))))

(event Custom|PawnEntredDBNO
  (bind _returnvalue (Abilities|TryActivateAbilitiesbyTag (Ability|GetAbilitySystemComponent (Variables|InternalData|GetAIPawn)) (Variables|Default|GetTag_RebirthAbilityTag) false))
  (if (not _returnvalue)
    (Fortnite|FortShippingLog (Variables|InternalData|GetAIPawn) (Utilities|String|ToString(Text) (Utilities|Text|FormatText "NSLOCTEXT(\"[54342A2F60202E80BAD36DAF84B08806]\", \"AFA097544A8316EE51C9BFB8858EABD4\", \"Pawn {a} tried to activate Rebirth Ability with Tag : {b} but ability could not be activated. Make sure a Tag has been set and the Ability can be used by this Pawn.\")" (Utilities|GetDisplayName (Variables|InternalData|GetAIPawn)) (GameplayTags|GetDebugStringfromGameplayTagContainer (Variables|Default|GetTag_RebirthAbilityTag)))) true))
  (Ability|EndAbility))

(event Ability|EventFailedToActivatePassiveAbility
  (params (ActorInfo :type Gameplay%20Ability%20Actor%20Info%20Structure)))

(event Ability|EventOnEndAbility
  (params (bWasCancelled :type Boolean))
  (EventDispatchers|UnbindEventfromOnPlayerStartDBNO (Variables|InternalData|GetAIPawn) (EventDispatchers|CreateEvent :!event Custom|PawnEntredDBNO)))
