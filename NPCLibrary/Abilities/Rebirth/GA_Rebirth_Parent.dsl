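; Parent gameplay ability that rebirths a downed (DBNO) NPC boss, with child abilities hooking OnPawnLanded.
; Pre-rebirth effects, cues and speech run until an event or delay starts transformation, optionally levitating the pawn, destroying builds above and swapping its outfit.
; Manual revive updates navigation and sets health and shields by playlist team size.
; Landing, entering water or skipping levitation enables health bars, readies engagement, plays a montage and ends the ability.
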(event Ability|EventActivateAbility
  (bind _aipawn (Variables|InternalData|GetAIPawn))
  (bind _returnvalue (Actor|GetComponentbyClass _aipawn "/Script/FortniteGame.FortPawnComponent_SpeechBubble"))
  ;[1 1 1] ActivateAbility - Validate Data
  (Ability|CommitAbility)
  (bind _outfortplayerpawn (CallFunction|GetAvatarasValidFortPlayerPawn)
    (:ValidFortPlayerPawn
      (bind _output_get (Variables|InternalData|SetAIPawn _outfortplayerpawn))
      (Utilities|IsValid (Pawn|GetController _output_get)
        (:Is%20Valid
          (bind _asfort_athena_aibot_controller (Utilities|Casting|CastToFortAthenaAIBotController (Pawn|GetController _output_get))
            (:then
              (Variables|InternalData|SetAIController _asfort_athena_aibot_controller)
              (CallFunction|InitVariables)
              (FortWeapon|ModAmmo (Variables|/Script/FortniteGame.FortPawn|GetCurrentWeapon _aipawn) 500)
              (CallFunction|StartPreRebirth)
              (bind _output_get_2 (Variables|InternalData|SetSpeechBubbleComponent _returnvalue))
              (Utilities|IsValid _output_get_2
                (:Is%20Valid
                  (SpeechBubble|StartSpeech _returnvalue (Variables|Default|GetRebirthSpeechQueryOnDBNO)))
                (:Is%20Not%20Valid
                  _))
              (bind (_gamestate _playlist) (Class|FortAsyncActionCurrentPlaylistReady|PlaylistDataReadyAsync)
                (:OnReady
                  (Variables|InternalData|SetPlaylistTeamSize (Variables|/Script/FortniteGame.FortPlaylist|GetMaxTeamSize _playlist)))))
            (:CastFailed
              _)))
        (:Is%20Not%20Valid
          _))
      (Ability|EndAbility))
    (:AvatarCastFailed
      _)))

(event Ability|EventOnEndAbility
  (params (bWasCancelled :type Boolean))
  (bind _aipawn (Variables|InternalData|GetAIPawn))
  ;[1 1 1] End Ability
  (Ability|SendGameplayEventtoActor 0 (Variables|Default|GetTag_EventEndTransformation))
  (Ability|RemoveGameplayEffectFromOwnerWithHandle (Variables|InternalData|GetHNDL_PreBirthGE))
  (Ability|ApplyGameplayEffectToOwner (Variables|Default|GetGE_PostRebirth))
  (CallFunction|ActivateMiniMapIcon)
  (GameplayTasks|EndTask (Variables|InternalData|GetTASK_DestroyBuilds))
  (Utilities|IsValid _aipawn
    (:Is%20Valid
      (bind _asbp_player_pawn_non_participant_boss (Utilities|Casting|CastToBP_PlayerPawn_NonParticipant_Boss _aipawn))
      (Class|BPPlayerPawnNonParticipantBoss|NPCEnableHealthBars _asbp_player_pawn_non_participant_boss))))

(event Custom|StartPreRebirth
  ;[1 1 1] Pre ReBirth (Before Transformation)
  (bind _returnvalue (Ability|ApplyGameplayEffectToOwner (Variables|Default|GetGE_PreRebirth) 0))
  (Variables|InternalData|SetHNDL_PreBirthGE _returnvalue)
  (Ability|AddGameplayCueToOwner (Variables|Default|GetGC_PreRebirthLoop))
  (Ability|AddGameplayCueToOwner (Variables|Default|GetGCL_DBNOPhase))
  (if (Variables|Default|GetUseEventToTriggerRebirth)
    (Ability|Async|WaitGameplayEventToActor (Variables|InternalData|GetAIPawn) (Variables|Default|GetEventTagForTransformation) true
      (:then
        (Utilities|FlowControl|Delay (ScalableFloat|GetValueatLevel (Variables|Default|GetFallbackDelayForTransformation)))
        (CallFunction|StartTransformation))
      (:EventReceived
        (CallFunction|StartTransformation)
        (Ability|RemoveGameplayCueFromOwner (Variables|Default|GetGCL_DBNOPhase))))
    (else
      (Utilities|FlowControl|Delay (ScalableFloat|GetValueatLevel (Variables|Default|GetDelayBeforeStartTransformation)))
      (CallFunction|StartTransformation))))

(event Custom|TestTriggerTransformation
  ;[0.2425 0.2615 1] Event to test the Transformation
  (Ability|SendGameplayEventtoActor (Variables|InternalData|GetAIPawn) (Variables|Default|GetEventTagForTransformation)))

(event Custom|SwapVisuals
  (bind _aipawn (Variables|InternalData|GetAIPawn))
  ;[1 1 1] Change Visuals
  (bind _returnvalue (GUID|GenerateGuidfromObject _aipawn)) ; Needed for Swap data
  (bind _output_get (Variables|Default|SetSwapId _returnvalue))
  (Class|FortPlayerPawn|SwapPawnCosmeticItemandVariant _aipawn _output_get (Variables|Default|GetFinalFormOutitData) false "Active" false) ; Change the outfit out to the night version
)

(event Custom|StartTransformation
  ;[1 1 1] Start Transformation
  (if (not (Variables|InternalData|GetTransformationStarted))
    (Ability|AddGameplayCueToOwner (Variables|Default|GetGCL_Rebirth))
    (if (ScalableFloat|GetValueasBool (Variables|Default|GetTryToTransformOnRebirth))
      (CallFunction|PushCharacterinAir)
      (else
        (CallFunction|StartSpeechBubbleonTransformation)))))

(event Custom|PushCharacterInAir
  (bind _aipawn (Variables|InternalData|GetAIPawn))
  ;[1 1 1] Start Transformation
  (Variables|InternalData|SetTransformationStarted true)
  (Ability|SendGameplayEventtoActor _aipawn (Variables|Default|GetTag_StartFloating))
  (CallFunction|TryDestroyBuildsAbove)
  (FortPawn|SetGravityMultiplier _aipawn 0.0)
  (Character|LaunchCharacter _aipawn "0.000000,0.000000,300.000000" true true)
  (Utilities|FlowControl|Delay 2.0)
  (Components|Movement|StopMovementImmediately (Variables|/Script/Engine.Character|GetCharacterMovement _aipawn))
  (Utilities|FlowControl|Delay 0.2)
  (CallFunction|StartSpeechBubbleonTransformation))

(event Custom|StartSpeechBubbleOnTransformation
  (bind _speechbubblecomponent (Variables|InternalData|GetSpeechBubbleComponent))
  ;[1 1 1] Start Transformation
  (Utilities|IsValid _speechbubblecomponent
    (:Is%20Valid
      (SpeechBubble|StartSpeech _speechbubblecomponent (Variables|Default|GetRebirthSpeechQueryBeforeTransformation)))
    (:Is%20Not%20Valid
      _))
  (Utilities|FlowControl|Delay (ScalableFloat|GetValueatLevel (Variables|Default|GetTimeToEnlarge)))
  (if (ScalableFloat|GetValueasBool (Variables|Default|GetSwapVisualsOnRebirth))
    (CallFunction|StartSwapVisualsonRebirth)
    (else
      (CallFunction|StartManualRevive))))

(event Custom|StartSwapVisualsOnRebirth
  ;[1 1 1] Start Transformation
  (Ability|SendGameplayEventtoActor (Variables|InternalData|GetAIPawn) (Variables|Default|GetTag_StartEnlarging))
  (Ability|ExecuteGameplayCueOnOwner (Variables|Default|GetGC_Transformation))
  (CallFunction|SwapVisuals)
  (Utilities|FlowControl|Delay 0.7)
  (CallFunction|StartManualRevive))

(event Custom|StartManualRevive
  (bind _navigationclassonrebirth (Variables|Default|GetNavigationClassOnRebirth))
  (bind _aipawn (Variables|InternalData|GetAIPawn))
  ;[1 1 1] Start Transformation
  (FortPlayerPawn|ManualRevivefromDBNO (Variables|InternalData|GetAIPawn) 0)
  (CallFunction|UpdateNewHealth)
  (Ability|RemoveGameplayCueFromOwner (Variables|Default|GetGC_PreRebirthLoop))
  (Ability|RemoveGameplayCueFromOwner (Variables|Default|GetGCL_Rebirth))
  (if (ScalableFloat|GetValueasBool (Variables|Default|GetTryToTransformOnRebirth))
    (FortPawn|SetGravityMultiplier _aipawn 1.0)
    (Variables|InternalData|SetIsFallingBackDown true)
    (GameplayTasks|EndTask (Variables|InternalData|GetTASK_DestroyBuilds))
    (Ability|SendGameplayEventtoActor _aipawn (Variables|Default|GetTag_EventEndTransformation))
    (else
      _))
  (Ability|RemoveGameplayEffectFromOwnerWithHandle (Variables|InternalData|GetHNDL_PreBirthGE))
  (if (Utilities|IsValidClass _navigationclassonrebirth)
    (Variables|/Script/AIModule.AIController|SetDefaultNavigationFilterClass _navigationclassonrebirth (Variables|InternalData|GetAIController))
    (else
      _))
  (if (Variables|Default|GetApplyNewNavAgentOnRebirth)
    (Variables|/Script/Engine.NavMovementComponent|SetNavAgentProps (Variables|Default|GetNavAgentOnRebirth) (Pawn|GetMovementComponent _aipawn))
    (else
      _))
  (EventDispatchers|BindEventtoOnPawnLanded _aipawn (EventDispatchers|CreateEvent :!event Custom|PawnHasLanded))
  (EventDispatchers|BindEventtoOnPawnEnterWater _aipawn (EventDispatchers|CreateEvent :!event Custom|PawnHasEnteredWater))
  (if (not (ScalableFloat|GetValueasBool (Variables|Default|GetTryToTransformOnRebirth)))
    (CallFunction|HandleRevivedWithoutTransformation)))

(event Custom|TryDestroyBuildsAbove
  ;[1 1 1] Destroy Builds on Transform
  (if (Variables|Default|Destruction|GetShouldDestroyAboveBuildsOnRebirth)
    (bind (_asynctaskproxy _targetdata _applicationtag) (Ability|Tasks|DoTargetSelectionContinuous "DestroyBuilds" (Variables|Default|Destruction|GetTag_ApplicationTagForDestroyBuilds) false)
      (:then
        (Variables|InternalData|SetTASK_DestroyBuilds _asynctaskproxy))
      (:Targeted
        (Ability|ApplyGameplayEffectContainer _targetdata _applicationtag)))))

(event Custom|PawnHasLanded
  (params (Hit :type Hit%20Result%20Structure%20%28by%20ref%29))
  (bind _aipawn (Variables|InternalData|GetAIPawn))
  (bind _speechbubblecomponent (Variables|InternalData|GetSpeechBubbleComponent))
  (bind _aipawn_2 (Variables|InternalData|GetAIPawn))
  ;[1 1 1] On Pawn Landed from Transformation
  (EventDispatchers|UnbindEventfromOnPawnLanded _aipawn (delegate Custom|PawnHasLanded))
  (EventDispatchers|UnbindEventfromOnPawnEnterWater _aipawn (delegate Custom|PawnHasEnteredWater))
  (Character|UnCrouch _aipawn)
  (CallFunction|OnPawnLanded)
  (Utilities|IsValid _aipawn_2
    (:Is%20Valid
      (bind _asbp_player_pawn_non_participant_boss (Utilities|Casting|CastToBP_PlayerPawn_NonParticipant_Boss _aipawn_2)
        (:then
          (Class|BPPlayerPawnNonParticipantBoss|NPCEnableHealthBars _asbp_player_pawn_non_participant_boss))
        (:CastFailed
          _))
      (Utilities|FlowControl|Delay 0.2)
      (bind _asbp_player_pawn_non_participant_boss_2 (Utilities|Casting|CastToBP_PlayerPawn_NonParticipant_Boss _aipawn_2)
        (:then
          (Class|BPPlayerPawnNonParticipantBoss|BossReadytoEngage _asbp_player_pawn_non_participant_boss_2))
        (:CastFailed
          _))
      (Ability|GameplayEffect|AddLooseGameplayTags (Variables|InternalData|GetAIPawn) (Variables|Default|GetTag_ForceSkipDBNO))
      (Ability|Tasks|PlayMontageAndWait "None" (Variables|Default|GetMontagePostRebirth)
        (:then
          (Utilities|Time|SetTimerbyEvent (delegate Custom|FailSafeEndAbility) (ScalableFloat|GetValueatLevel (Variables|Default|GetFailSafeTimer)))
          (Utilities|IsValid _speechbubblecomponent
            (:Is%20Valid
              (SpeechBubble|StopSpeech _speechbubblecomponent)
              (SpeechBubble|StartSpeech _speechbubblecomponent (Variables|Default|GetRebirthSpeechQueryFinishedRebirth)))))
        (:OnCompleted
          _)
        (:OnBlendOut
          (Utilities|FlowControl|Delay 0.3))
        (:OnInterrupted
          _)
        (:OnCancelled
          _))
      (Ability|EndAbility))
    (:Is%20Not%20Valid
      _)))

(event Custom|PawnHasEnteredWater
  (params (FortPlayerPawn :type Fort%20Player%20Pawn%20Object%20Reference))
  (bind _aipawn (Variables|InternalData|GetAIPawn))
  (bind _speechbubblecomponent (Variables|InternalData|GetSpeechBubbleComponent))
  (bind _aipawn_2 (Variables|InternalData|GetAIPawn))
  ;[1 1 1] On Pawn Landed from Transformation
  (EventDispatchers|UnbindEventfromOnPawnLanded _aipawn (delegate Custom|PawnHasLanded))
  (EventDispatchers|UnbindEventfromOnPawnEnterWater _aipawn (delegate Custom|PawnHasEnteredWater))
  (Character|UnCrouch _aipawn)
  (CallFunction|OnPawnLanded)
  (Utilities|IsValid _aipawn_2
    (:Is%20Valid
      (bind _asbp_player_pawn_non_participant_boss (Utilities|Casting|CastToBP_PlayerPawn_NonParticipant_Boss _aipawn_2)
        (:then
          (Class|BPPlayerPawnNonParticipantBoss|NPCEnableHealthBars _asbp_player_pawn_non_participant_boss))
        (:CastFailed
          _))
      (Utilities|FlowControl|Delay 0.2)
      (bind _asbp_player_pawn_non_participant_boss_2 (Utilities|Casting|CastToBP_PlayerPawn_NonParticipant_Boss _aipawn_2)
        (:then
          (Class|BPPlayerPawnNonParticipantBoss|BossReadytoEngage _asbp_player_pawn_non_participant_boss_2))
        (:CastFailed
          _))
      (Ability|GameplayEffect|AddLooseGameplayTags (Variables|InternalData|GetAIPawn) (Variables|Default|GetTag_ForceSkipDBNO))
      (Ability|Tasks|PlayMontageAndWait "None" (Variables|Default|GetMontagePostRebirth)
        (:then
          (Utilities|Time|SetTimerbyEvent (delegate Custom|FailSafeEndAbility) (ScalableFloat|GetValueatLevel (Variables|Default|GetFailSafeTimer)))
          (Utilities|IsValid _speechbubblecomponent
            (:Is%20Valid
              (SpeechBubble|StopSpeech _speechbubblecomponent)
              (SpeechBubble|StartSpeech _speechbubblecomponent (Variables|Default|GetRebirthSpeechQueryFinishedRebirth)))))
        (:OnCompleted
          _)
        (:OnBlendOut
          (Utilities|FlowControl|Delay 0.3))
        (:OnInterrupted
          _)
        (:OnCancelled
          _))
      (Ability|EndAbility))
    (:Is%20Not%20Valid
      _)))

(event Custom|FailSafeEndAbility
  ;[1 1 1] On Pawn Landed from Transformation
  (Ability|EndAbility))

(event Custom|HandleRevivedWithoutTransformation
  (bind _aipawn (Variables|InternalData|GetAIPawn))
  (bind _speechbubblecomponent (Variables|InternalData|GetSpeechBubbleComponent))
  (bind _aipawn_2 (Variables|InternalData|GetAIPawn))
  ;[1 1 1] On Pawn Landed from Transformation
  (EventDispatchers|UnbindEventfromOnPawnLanded _aipawn (delegate Custom|PawnHasLanded))
  (EventDispatchers|UnbindEventfromOnPawnEnterWater _aipawn (delegate Custom|PawnHasEnteredWater))
  (Character|UnCrouch _aipawn)
  (CallFunction|OnPawnLanded)
  (Utilities|IsValid _aipawn_2
    (:Is%20Valid
      (bind _asbp_player_pawn_non_participant_boss (Utilities|Casting|CastToBP_PlayerPawn_NonParticipant_Boss _aipawn_2)
        (:then
          (Class|BPPlayerPawnNonParticipantBoss|NPCEnableHealthBars _asbp_player_pawn_non_participant_boss))
        (:CastFailed
          _))
      (Utilities|FlowControl|Delay 0.2)
      (bind _asbp_player_pawn_non_participant_boss_2 (Utilities|Casting|CastToBP_PlayerPawn_NonParticipant_Boss _aipawn_2)
        (:then
          (Class|BPPlayerPawnNonParticipantBoss|BossReadytoEngage _asbp_player_pawn_non_participant_boss_2))
        (:CastFailed
          _))
      (Ability|GameplayEffect|AddLooseGameplayTags (Variables|InternalData|GetAIPawn) (Variables|Default|GetTag_ForceSkipDBNO))
      (Ability|Tasks|PlayMontageAndWait "None" (Variables|Default|GetMontagePostRebirth)
        (:then
          (Utilities|Time|SetTimerbyEvent (delegate Custom|FailSafeEndAbility) (ScalableFloat|GetValueatLevel (Variables|Default|GetFailSafeTimer)))
          (Utilities|IsValid _speechbubblecomponent
            (:Is%20Valid
              (SpeechBubble|StopSpeech _speechbubblecomponent)
              (SpeechBubble|StartSpeech _speechbubblecomponent (Variables|Default|GetRebirthSpeechQueryFinishedRebirth)))))
        (:OnCompleted
          _)
        (:OnBlendOut
          (Utilities|FlowControl|Delay 0.3))
        (:OnInterrupted
          _)
        (:OnCancelled
          _))
      (Ability|EndAbility))
    (:Is%20Not%20Valid
      _)))

(fn ActivateMiniMapIcon
  (bind _aipawn (Variables|InternalData|GetAIPawn))
  (bind _returnvalue (Actor|GetComponentbyClass _aipawn "/Script/FortniteGame.AthenaSpecialActorComponent"))
  (Utilities|IsValid _aipawn
    (:Is%20Valid
      (Utilities|IsValid _returnvalue
        (:Is%20Valid
          (Components|Activation|Activate _returnvalue false))
        (:Is%20Not%20Valid
          _))
      (return))
    (:Is%20Not%20Valid
      _)))

(fn GetCustomAbilitySourceTransform
  (bind (_returnvalue_x _returnvalue_y _returnvalue_z) (Math|Vector|BreakVector (Transformation|GetActorLocation (CallFunction|GetAvatarasFortPlayerPawn))))
  (return _returnvalue_x _returnvalue_y (+ _returnvalue_z (ScalableFloat|GetValueatLevel (Variables|Default|Destruction|GetDestructionOffset))) (Transformation|GetActorRotation (CallFunction|GetAvatarasFortPlayerPawn)) "1.000000,1.000000,1.000000"))

(fn InitVariables
  (Variables|InternalData|SetIsFallingBackDown false)
  (Variables|InternalData|SetTransformationStarted false))

(fn OnPawnLanded
  ;[1 0.1667 0.1667] Used in Child Abilities
  (return))

(fn UpdateNewHealth
  (bind _playlistteamsize (Variables|InternalData|GetPlaylistTeamSize))
  (bind _returnvalue (Math|Conversions|ToFloat(Integer) _playlistteamsize))
  (bind _newmaxhealth (Variables|Default|GetNewMaxHealth))
  (bind _returnvalue_2 (ScalableFloat|GetValueatLevel _newmaxhealth _returnvalue))
  (bind _newmaxshields (Variables|Default|GetNewMaxShields))
  (bind _returnvalue_3 (ScalableFloat|GetValueatLevel _newmaxshields _returnvalue))
  (bind _aipawn (Variables|InternalData|GetAIPawn))
  (if (< (Variables|InternalData|GetPlaylistTeamSize) 1)
    (Variables|InternalData|SetPlaylistTeamSize 1)
    (elif (> (Variables|InternalData|GetPlaylistTeamSize) 4)
      (Variables|InternalData|SetPlaylistTeamSize 4)
      (FortPawn|SetMaxHealth _aipawn (Math|Float|SafeDivide _returnvalue_2 (ScalableFloat|GetValueatLevel (Variables|Default|GetPreviousMaxHealthMultiplier) _returnvalue)))
      (FortPawn|SetHealth _aipawn _returnvalue_2)
      (FortPawn|SetMaxShield _aipawn _returnvalue_3)
      (FortPawn|SetShield _aipawn _returnvalue_3)
      (return))))
